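# Ambient spot sound cue for a light-beam data wall prop, layering looping secret-facility computer sounds.
# Two mixed layers each pick a random "calculating" or "idling" loop with random pitch modulation.
# Looping envelopers slowly sweep the layers' pitch in opposite directions, so they drift against each other.
# Only the first two quality slots are wired; small-spot attenuation and a distance-scaled concurrency override apply.
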SoundCue /Hera/Audio/Props/SC_Spot_LightBeam_DataWall_A
root=QualityLevel_1
nodes [
 {
  id=Enveloper_0
  children [Modulator_2]
  properties {
   bLoopIndefinitely=true
   pitchCurve {editorCurveData {keys [
      {time=0}
      {arriveTangent=0.005, interpMode=RCIM_Cubic, leaveTangent=0.005, tangentMode=RCTM_SmartAuto, time=10, value=0.9}
      <at 20 1.1>
      {time=30}
     ]}}
   volumeCurve {editorCurveData {keys [<at 0 1>]}}
  }
 }
 {
  id=Enveloper_1
  children [Modulator_0]
  properties {
   bLoopIndefinitely=true
   pitchCurve {editorCurveData {keys [
      {time=0}
      {arriveTangent=0.005, interpMode=RCIM_Cubic, leaveTangent=0.005, tangentMode=RCTM_SmartAuto, time=10, value=1.1}
      <at 20 0.9>
      {time=30}
     ]}}
   volumeCurve {editorCurveData {keys [<at 0 1>]}}
  }
 }
 {id=Mixer_0, children [Enveloper_1,Enveloper_0], properties {inputVolume [1,1]}}
 {id=Modulator_0, children [Random_0], properties {pitchMax=1.1, pitchMin=0.9}}
 {id=Modulator_2, children [Random_1], properties {pitchMax=1.1, pitchMin=0.9}}
 {id=QualityLevel_1, children [Mixer_0,Mixer_0,null,null]}
 {
  id=Random_0
  children [WavePlayer_0,WavePlayer_1,WavePlayer_2,WavePlayer_4,WavePlayer_5,WavePlayer_6,WavePlayer_7,WavePlayer_8]
  properties {weights [1,1,1,1,1,1,1,1]}
 }
 {
  id=Random_1
  children [WavePlayer_9, WavePlayer_10, WavePlayer_11, WavePlayer_12, WavePlayer_13, WavePlayer_14, WavePlayer_15, WavePlayer_16, WavePlayer_17]
  properties {weights [1,1,1,1,1,1,1,1,1]}
 }
 {
  id=WavePlayer_0
  properties {bLooping=true, soundWaveAssetPtr=/Hera/Audio/POI/SecretFacility/Spots/Spot_SecretFacility_Computer_Calculating_Loop_01}
 }
 {
  id=WavePlayer_1
  properties {bLooping=true, soundWaveAssetPtr=/Hera/Audio/POI/SecretFacility/Spots/Spot_SecretFacility_Computer_Calculating_Loop_02}
 }
 {
  id=WavePlayer_10
  properties {bLooping=true, soundWaveAssetPtr=/Hera/Audio/POI/SecretFacility/Spots/Spot_SecretFacility_Computer_Idling_Loop_02}
 }
 {
  id=WavePlayer_11
  properties {bLooping=true, soundWaveAssetPtr=/Hera/Audio/POI/SecretFacility/Spots/Spot_SecretFacility_Computer_Idling_Loop_03}
 }
 {
  id=WavePlayer_12
  properties {bLooping=true, soundWaveAssetPtr=/Hera/Audio/POI/SecretFacility/Spots/Spot_SecretFacility_Computer_Idling_Loop_04}
 }
 {
  id=WavePlayer_13
  properties {bLooping=true, soundWaveAssetPtr=/Hera/Audio/POI/SecretFacility/Spots/Spot_SecretFacility_Computer_Idling_Loop_05}
 }
 {
  id=WavePlayer_14
  properties {bLooping=true, soundWaveAssetPtr=/Hera/Audio/POI/SecretFacility/Spots/Spot_SecretFacility_Computer_Idling_Loop_06}
 }
 {
  id=WavePlayer_15
  properties {bLooping=true, soundWaveAssetPtr=/Hera/Audio/POI/SecretFacility/Spots/Spot_SecretFacility_Computer_Idling_Loop_07}
 }
 {
  id=WavePlayer_16
  properties {bLooping=true, soundWaveAssetPtr=/Hera/Audio/POI/SecretFacility/Spots/Spot_SecretFacility_Computer_Idling_Loop_08}
 }
 {
  id=WavePlayer_17
  properties {bLooping=true, soundWaveAssetPtr=/Hera/Audio/POI/SecretFacility/Spots/Spot_SecretFacility_Computer_Idling_Loop_09}
 }
 {
  id=WavePlayer_2
  properties {bLooping=true, soundWaveAssetPtr=/Hera/Audio/POI/SecretFacility/Spots/Spot_SecretFacility_Computer_Calculating_Loop_03}
 }
 {
  id=WavePlayer_4
  properties {bLooping=true, soundWaveAssetPtr=/Hera/Audio/POI/SecretFacility/Spots/Spot_SecretFacility_Computer_Calculating_Loop_05}
 }
 {
  id=WavePlayer_5
  properties {bLooping=true, soundWaveAssetPtr=/Hera/Audio/POI/SecretFacility/Spots/Spot_SecretFacility_Computer_Calculating_Loop_06}
 }
 {
  id=WavePlayer_6
  properties {bLooping=true, soundWaveAssetPtr=/Hera/Audio/POI/SecretFacility/Spots/Spot_SecretFacility_Computer_Calculating_Loop_07}
 }
 {
  id=WavePlayer_7
  properties {bLooping=true, soundWaveAssetPtr=/Hera/Audio/POI/SecretFacility/Spots/Spot_SecretFacility_Computer_Calculating_Loop_08}
 }
 {
  id=WavePlayer_8
  properties {bLooping=true, soundWaveAssetPtr=/Hera/Audio/POI/SecretFacility/Spots/Spot_SecretFacility_Computer_Calculating_Loop_09}
 }
 {
  id=WavePlayer_9
  properties {bLooping=true, soundWaveAssetPtr=/Hera/Audio/POI/SecretFacility/Spots/Spot_SecretFacility_Computer_Idling_Loop_01}
 }
]
properties {
 attenuationSettings=/Hera/Audio/Attenuations/ATT_Hera_Spot_Small
 bEnableSubmixSends=false
 bOverrideConcurrency=true
 maxDistance=1100
 soundClassObject=&Fort_Ambient_Props_Athena
 volumeMultiplier=1
 attenuationOverrides {
  attenuationShape=Cone
  bEnableListenerFocus=true
  bEnableOcclusion=true
  coneSphereFalloffDistance=350
  coneSphereRadius=250
  falloffDistance=1300
  focusVolumeAttenuation=1.1
  nonFocusVolumeAttenuation=0.6
  occlusionInterpolationTime=1
  occlusionLowPassFilterFrequency=750
  occlusionTraceChannel=ECC_GameTraceChannel11
  occlusionVolumeAttenuation=0.5
  reverbWetLevelMin=0.6905
  stereoSpread=100
  attenuationShapeExtents=<650 90 400>
 }
 concurrencyOverrides {maxCount=3, volumeScaleMode=Distance}
}
