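# Looping ambient prop cue for a MedPod light beam, built from holographic table wave loops and spatialized with cone attenuation, occlusion and listener focus.
# It mixes a base loop with a layer that distance-crossfades one of three random close loops into a distant loop.
# Nested looping envelopes slowly drift each layer's pitch and volume.
# Quality-level nodes keep the base in all four slots and the crossfade layer only in the first two.
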SoundCue /Hera/Audio/Props/SC_Spot_LightBeam_MedPod_A
root=Mixer_0
nodes [
 {
  id=DistanceCrossFade_0
  children [Random_0, WavePlayer_14]
  properties {crossFadeInput [
    {fadeInDistanceEnd=0, fadeInDistanceStart=0, fadeOutDistanceEnd=1000, fadeOutDistanceStart=500, volume=1}
    {fadeInDistanceEnd=0, fadeInDistanceStart=0, fadeOutDistanceEnd=1500, fadeOutDistanceStart=1000, volume=0.7}
   ]}
 }
 {
  id=Enveloper_0
  children [Mixer_1]
  properties {
   bLoop=true
   bLoopIndefinitely=true
   loopEnd=30
   pitchCurve {editorCurveData {keys [{arriveTangent=-0.0033, interpMode=RCIM_Cubic, time=0}, <at 15 1.05 RCIM_Cubic>, {interpMode=RCIM_Cubic, leaveTangent=0.0033, time=30}]}}
   volumeCurve {editorCurveData {keys [<at 0 1>]}}
  }
 }
 {
  id=Enveloper_3
  children [Enveloper_0]
  properties {
   bLoop=true
   bLoopIndefinitely=true
   loopEnd=20
   pitchCurve {editorCurveData {keys [{time=0}]}}
   volumeCurve {editorCurveData {keys [
      {arriveTangent=-0.01, interpMode=RCIM_Cubic, time=0, value=1}
      {arriveTangent=-0.0665, interpMode=RCIM_Cubic, leaveTangent=-0.0665, time=2.3014, value=0.8437}
      <at 4.0452 0.7311>
      {arriveTangent=0.0026, interpMode=RCIM_Cubic, leaveTangent=0.0026, time=6.4209, value=0.8395}
      {arriveTangent=0.0118, interpMode=RCIM_Cubic, leaveTangent=0.0118, time=11.9487, value=0.7516}
      {interpMode=RCIM_Cubic, leaveTangent=0.01, time=20, value=1}
     ]}}
  }
 }
 {
  id=Enveloper_4
  children [DistanceCrossFade_0]
  properties {
   bLoop=true
   bLoopIndefinitely=true
   loopEnd=30
   pitchCurve {editorCurveData {keys [{arriveTangent=-0.0033, interpMode=RCIM_Cubic, time=0}, <at 15 1.05 RCIM_Cubic>, {interpMode=RCIM_Cubic, leaveTangent=0.0033, time=30}]}}
   volumeCurve {editorCurveData {keys [<at 0 1>]}}
  }
 }
 {
  id=Enveloper_5
  children [Enveloper_4]
  properties {
   bLoop=true
   bLoopIndefinitely=true
   loopEnd=20
   pitchCurve {editorCurveData {keys [{time=0}]}}
   volumeCurve {editorCurveData {keys [
      {arriveTangent=-0.01, interpMode=RCIM_Cubic, time=0, value=1}
      {arriveTangent=-0.0572, interpMode=RCIM_Cubic, leaveTangent=-0.0572, time=2.3014, value=0.8437}
      <at 4.0652 0.7676>
      {arriveTangent=-0.002, interpMode=RCIM_Cubic, leaveTangent=-0.002, time=6.4209, value=0.8395}
      {arriveTangent=0.0118, interpMode=RCIM_Cubic, leaveTangent=0.0118, time=11.9487, value=0.7516}
      {interpMode=RCIM_Cubic, leaveTangent=0.01, time=20, value=1}
     ]}}
  }
 }
 {id=Mixer_0, children [QualityLevel_1, QualityLevel_2], properties {inputVolume [1,1]}}
 {id=Mixer_1, children [WavePlayer_10], properties {inputVolume [1]}}
 {id=Modulator_0, children [Enveloper_3], properties {volumeMax=1.3, volumeMin=1.3}}
 {id=Modulator_2, children [Enveloper_5], properties {volumeMax=1.3, volumeMin=1.3}}
 {id=QualityLevel_1, children [Modulator_0,Modulator_0,Modulator_0,Modulator_0]}
 {id=QualityLevel_2, children [Modulator_2,Modulator_2,null,null]}
 {id=Random_0, children [WavePlayer_11, WavePlayer_12, WavePlayer_13], properties {weights [1,1,1]}}
 {id=WavePlayer_10, properties {bLooping=true, soundWaveAssetPtr=&/Spot_LightBeam_Holographic_Table_Base_Loop_01}}
 {id=WavePlayer_11, properties {bLooping=true, soundWaveAssetPtr=&/Spot_LightBeam_Holographic_Table_Close_Loop_01}}
 {id=WavePlayer_12, properties {bLooping=true, soundWaveAssetPtr=&/Spot_LightBeam_Holographic_Table_Close_Loop_02}}
 {id=WavePlayer_13, properties {bLooping=true, soundWaveAssetPtr=&/Spot_LightBeam_Holographic_Table_Close_Loop_03}}
 {id=WavePlayer_14, properties {bLooping=true, soundWaveAssetPtr=&/Spot_LightBeam_Holographic_Table_Distant_Loop_01}}
]
properties {
 attenuationSettings=/Hera/Audio/Attenuations/ATT_Hera_Spot_Medium
 bEnableSubmixSends=false
 bOverrideConcurrency=true
 maxDistance=1500
 soundClassObject=&Fort_Ambient_Props_Athena
 volumeMultiplier=0.5
 attenuationOverrides {
  attenuationShape=Cone
  bEnableListenerFocus=true
  bEnableOcclusion=true
  coneSphereFalloffDistance=350
  coneSphereRadius=250
  falloffDistance=1300
  focusVolumeAttenuation=1.1
  nonFocusVolumeAttenuation=0.6
  occlusionInterpolationTime=1
  occlusionLowPassFilterFrequency=750
  occlusionTraceChannel=ECC_GameTraceChannel11
  occlusionVolumeAttenuation=0.5
  reverbWetLevelMin=0.6905
  stereoSpread=100
  attenuationShapeExtents=<650 90 400>
 }
 concurrencyOverrides {maxCount=3, volumeScaleMode=Distance}
}
