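# Looping spot ambience cue for a movie theater projector machine, on the outdoor ambient sound class.
# A looping wave feeds two nested, endlessly looping Envelopers: the inner varies pitch and the outer slowly dips volume, on different cycle lengths so the loop drifts.
# A root Modulator raises the pitch by a fixed factor.
# Overrides add cone attenuation with listener focus and occlusion filtering, plus a small concurrency cap with distance-based volume scaling.
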SoundCue /Hera/Audio/Props/SC_Spot_Machine_MovieTheaterProjector
root=Modulator_4
nodes [
 {
  id=Enveloper_0
  children [WavePlayer_1]
  properties {
   bLoop=true
   bLoopIndefinitely=true
   loopEnd=30
   pitchCurve {editorCurveData {keys [
      {arriveTangent=-0.0013, interpMode=RCIM_Cubic, leaveTangent=-0.0013, time=0}
      <at 15 0.98 RCIM_Cubic>
      {arriveTangent=0.0013, interpMode=RCIM_Cubic, leaveTangent=0.0013, time=30}
     ]}}
   volumeCurve {editorCurveData {keys [<at 0 1>]}}
  }
 }
 {
  id=Enveloper_3
  children [Enveloper_0]
  properties {
   bLoop=true
   bLoopIndefinitely=true
   loopEnd=20
   volumeMin=0.7
   pitchCurve {editorCurveData {keys [{time=0}]}}
   volumeCurve {editorCurveData {keys [
      {arriveTangent=-0.01, interpMode=RCIM_Cubic, leaveTangent=-0.01, time=0, value=1}
      <at 10 0.9 RCIM_Cubic>
      {arriveTangent=0.01, interpMode=RCIM_Cubic, leaveTangent=0.01, time=20, value=1}
     ]}}
  }
 }
 {id=Modulator_4, children [Enveloper_3], properties {pitchMax=1.4, pitchMin=1.4}}
 {id=WavePlayer_1, properties {bLooping=true, soundWaveAssetPtr=&/Spot_Machine_MovieTheaterProjector_Loop_01}}
]
properties {
 bEnableSubmixSends=false
 bOverrideAttenuation=true
 bOverrideConcurrency=true
 maxDistance=1650
 soundClassObject=&Fort_Ambient_Outdoors_Athena
 volumeMultiplier=1.5
 attenuationOverrides {
  attenuationShape=Cone
  bEnableListenerFocus=true
  coneSphereFalloffDistance=1000
  coneSphereRadius=500
  falloffDistance=1000
  focusVolumeAttenuation=1.1
  nonFocusVolumeAttenuation=0.6
  occlusionInterpolationTime=1
  occlusionLowPassFilterFrequency=750
  occlusionTraceChannel=ECC_GameTraceChannel11
  occlusionVolumeAttenuation=0.5
  reverbWetLevelMin=0.6905
  stereoSpread=100
  attenuationShapeExtents=<650 90 2000>
 }
 concurrencyOverrides {maxCount=3, volumeScaleMode=Distance}
}
