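# Ambient spot sound cue for a golf vending machine prop, layering a steady hum with intermittent electronic and mechanical chatter.
# A looping refrigerator hum pitched up an octave plays at every quality level.
# The first two quality levels add delayed loops of a select blip and randomized disk-drive servo and writing one-shots, the latter pitch-varied.
# Attenuation overrides use a capsule with occlusion and listener focus, alongside a concurrency limit.
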SoundCue /Hera/Audio/Props/SC_Spot_Recreation_GolfVendingMachine_A
root=Mixer_0
nodes [
 {id=Delay_1, children [Modulator_3], properties {delayMax=2, delayMin=2}}
 {id=Delay_2, children [Looping_1], properties {delayMax=0.2, delayMin=0.2}}
 {id=Delay_3, children [Random_4], properties {delayMax=3, delayMin=1}}
 {id=Delay_4, children [Looping_2], properties {delayMax=0.2, delayMin=0.2}}
 {id=Delay_5, children [Random_3], properties {delayMax=3, delayMin=0.5}}
 {id=Delay_6, children [Looping_3], properties {delayMax=0.2, delayMin=0.2}}
 {id=Looping_1, children [Delay_1]}
 {id=Looping_2, children [Delay_3]}
 {id=Looping_3, children [Delay_5]}
 {id=Mixer_0, children [QualityLevel_0, QualityLevel_1], properties {inputVolume [1,1]}}
 {id=Mixer_1, children [Delay_2,Modulator_0,Modulator_5], properties {inputVolume [0.3,1,1]}}
 {id=Mixer_2, children [Modulator_1], properties {inputVolume [0.5]}}
 {id=Modulator_0, children [Delay_4], properties {pitchMax=1.1, pitchMin=0.9}}
 {id=Modulator_1, children [WavePlayer_3], properties {pitchMax=2, pitchMin=2}}
 {id=Modulator_3, children [WavePlayer_9]}
 {id=Modulator_5, children [Delay_6], properties {pitchMax=1.1, pitchMin=0.9}}
 {id=QualityLevel_0, children [Mixer_2,Mixer_2,Mixer_2,Mixer_2]}
 {id=QualityLevel_1, children [Mixer_1,Mixer_1,null,null]}
 {
  id=Random_3
  children [WavePlayer_17, WavePlayer_18, WavePlayer_19, WavePlayer_20, WavePlayer_21, WavePlayer_22]
  properties {weights [1,1,1,1,1,1]}
 }
 {id=Random_4, children [WavePlayer_23, WavePlayer_24, WavePlayer_25, WavePlayer_26], properties {weights [1,1,1,1]}}
 {id=WavePlayer_17, properties {soundWaveAssetPtr=&/Spot_Computer_Disk_Drive_Short_Servo_Resonance_01}}
 {id=WavePlayer_18, properties {soundWaveAssetPtr=&/Spot_Computer_Disk_Drive_Short_Servo_Resonance_02}}
 {id=WavePlayer_19, properties {soundWaveAssetPtr=&/Spot_Computer_Disk_Drive_Short_Servo_Resonance_03}}
 {id=WavePlayer_20, properties {soundWaveAssetPtr=&/Spot_Computer_Disk_Drive_Writing_01}}
 {id=WavePlayer_21, properties {soundWaveAssetPtr=&/Spot_Computer_Disk_Drive_Writing_02}}
 {id=WavePlayer_22, properties {soundWaveAssetPtr=&/Spot_Computer_Disk_Drive_Writing_03}}
 {id=WavePlayer_23, properties {soundWaveAssetPtr=&/Spot_Computer_Disk_Drive_Short_Servo_Movement_01}}
 {id=WavePlayer_24, properties {soundWaveAssetPtr=&/Spot_Computer_Disk_Drive_Short_Servo_Movement_02}}
 {id=WavePlayer_25, properties {soundWaveAssetPtr=&/Spot_Computer_Disk_Drive_Short_Servo_Movement_03}}
 {id=WavePlayer_26, properties {soundWaveAssetPtr=&/Spot_Computer_Disk_Drive_Short_Servo_Movement_04}}
 {id=WavePlayer_3, properties {bLooping=true, soundWaveAssetPtr=/AmbientAudioContent/Waves/SpottedLoops/Spot_OilRig_Refridgerator_Hum_02_Loop}}
 {id=WavePlayer_9, properties {soundWaveAssetPtr=/AmbientAudioContent/Waves/OneShots/Spot_DVR_CamOptions_Select}}
]
properties {
 attenuationSettings=/Hera/Audio/Attenuations/ATT_Hera_Spot_Medium
 bOverrideAttenuation=true
 bOverrideConcurrency=true
 maxDistance=1184.1239
 soundClassObject=&Fort_Ambient_Props_Athena
 volumeMultiplier=1
 attenuationOverrides {
  attenuationShape=Capsule
  bEnableListenerFocus=true
  bEnableOcclusion=true
  falloffDistance=1000
  focusVolumeAttenuation=1.1
  nonFocusVolumeAttenuation=0.6
  occlusionInterpolationTime=1
  occlusionLowPassFilterFrequency=750
  occlusionTraceChannel=ECC_GameTraceChannel11
  reverbWetLevelMin=0.6905
  stereoSpread=100
  attenuationShapeExtents=<184.1238 100.2439 400>
 }
 concurrencyOverrides {maxCount=3, volumeScaleMode=Distance}
}
