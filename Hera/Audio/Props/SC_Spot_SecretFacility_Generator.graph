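# Looping ambient spot sound for a generator prop in a secret facility, routed to an outdoor ambient sound class at reduced cue volume.
# A looping wave feeds an endlessly looping Enveloper that gently dips and restores volume, then a Modulator with pitch fixed at unity.
# The cue overrides attenuation with a capsule shape and listener focus, and overrides concurrency to limit simultaneous instances with distance-based volume scaling.
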SoundCue /Hera/Audio/Props/SC_Spot_SecretFacility_Generator
root=Modulator_4
nodes [
 {
  id=Enveloper_3
  children [WavePlayer_0]
  properties {
   bLoop=true
   bLoopIndefinitely=true
   loopEnd=20
   pitchCurve {editorCurveData {keys [{time=0}]}}
   volumeCurve {editorCurveData {keys [
      {arriveTangent=-0.01, interpMode=RCIM_Cubic, leaveTangent=-0.01, time=0, value=1}
      <at 10 0.9 RCIM_Cubic>
      {arriveTangent=0.01, interpMode=RCIM_Cubic, leaveTangent=0.01, time=20, value=1}
     ]}}
  }
 }
 {id=Modulator_4, children [Enveloper_3], properties {pitchMax=1, pitchMin=1}}
 {id=WavePlayer_0, properties {bLooping=true, soundWaveAssetPtr=&/Spot_SecretFacility_Generator_Loop_01}}
]
properties {
 bEnableSubmixSends=false
 bOverrideAttenuation=true
 bOverrideConcurrency=true
 maxDistance=2000
 soundClassObject=&Fort_Ambient_Outdoors_Athena
 volumeMultiplier=0.5
 attenuationOverrides {
  attenuationShape=Capsule
  bEnableListenerFocus=true
  coneSphereFalloffDistance=300
  coneSphereRadius=200
  falloffDistance=1000
  focusVolumeAttenuation=1.1
  nonFocusVolumeAttenuation=0.6
  occlusionInterpolationTime=1
  occlusionLowPassFilterFrequency=750
  occlusionTraceChannel=ECC_GameTraceChannel11
  occlusionVolumeAttenuation=0.5
  reverbWetLevelMin=0.6905
  stereoSpread=100
  attenuationShapeExtents=<1000 500 400>
 }
 concurrencyOverrides {maxCount=3, volumeScaleMode=Distance}
}
