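# Looping ambient spot cue for the SilverPopCorn DailiesCart prop, built from vintage computer idling sounds.
# A random node picks one of seven looping waves, each trimmed by its own mixer, then a modulator varies pitch and volume.
# An indefinitely looping envelope adds a slight volume dip; only the first two quality-level inputs are wired.
# Overrides add occlusion, listener focus and a small distance-scaled concurrency cap.
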SoundCue /Hera/Audio/Props/SC_Spot_SilverPopCorn_DailiesCart_A
root=QualityLevel_1
nodes [
 {
  id=Enveloper_0
  children [Modulator_0]
  properties {
   bLoop=true
   bLoopIndefinitely=true
   loopEnd=20
   pitchCurve {editorCurveData {keys [{time=0}]}}
   volumeCurve {editorCurveData {keys [<at -0.0028 1>,<at 0 1>,<at 10 0.95>,<at 20 1>]}}
  }
 }
 {id=Mixer_0, children [WavePlayer_5], properties {inputVolume [0.5]}}
 {id=Mixer_3, children [WavePlayer_8], properties {inputVolume [0.55]}}
 {id=Mixer_4, children [WavePlayer_1], properties {inputVolume [0.5]}}
 {id=Mixer_5, children [WavePlayer_2], properties {inputVolume [0.7]}}
 {id=Mixer_6, children [WavePlayer_0], properties {inputVolume [0.7]}}
 {id=Mixer_7, children [WavePlayer_6], properties {inputVolume [0.8]}}
 {id=Mixer_8, children [WavePlayer_7], properties {inputVolume [0.8]}}
 {id=Modulator_0, children [Random_0], properties {pitchMax=1.1, pitchMin=0.9, volumeMax=1, volumeMin=0.9}}
 {id=QualityLevel_1, children [Enveloper_0,Enveloper_0,null,null]}
 {id=Random_0, children [Mixer_6,Mixer_4,Mixer_5,Mixer_0,Mixer_7,Mixer_8,Mixer_3], properties {weights [1,1,1,1,1,1,1]}}
 {id=WavePlayer_0, properties {bLooping=true, soundWaveAssetPtr=&/Spot_Vintage_Computer_Idling_Loop_01}}
 {id=WavePlayer_1, properties {bLooping=true, soundWaveAssetPtr=&/Spot_Vintage_Computer_Idling_Loop_02}}
 {id=WavePlayer_2, properties {bLooping=true, soundWaveAssetPtr=&/Spot_Vintage_Computer_Idling_Loop_03}}
 {id=WavePlayer_5, properties {bLooping=true, soundWaveAssetPtr=&/Spot_Vintage_Computer_Idling_Loop_04}}
 {id=WavePlayer_6, properties {bLooping=true, soundWaveAssetPtr=&/Spot_Vintage_Computer_Idling_Loop_05}}
 {id=WavePlayer_7, properties {bLooping=true, soundWaveAssetPtr=&/Spot_Vintage_Computer_Idling_Loop_06}}
 {id=WavePlayer_8, properties {bLooping=true, soundWaveAssetPtr=&/Spot_Vintage_Computer_Idling_Loop_07}}
]
properties {
 bEnableSubmixSends=false
 bOverrideAttenuation=true
 bOverrideConcurrency=true
 maxDistance=1200
 soundClassObject=&Fort_Ambient_Props_Athena
 volumeMultiplier=0.7
 attenuationOverrides {
  bEnableListenerFocus=true
  bEnableOcclusion=true
  falloffDistance=550
  focusVolumeAttenuation=1.1
  nonFocusVolumeAttenuation=0.6
  occlusionInterpolationTime=1
  occlusionLowPassFilterFrequency=750
  occlusionTraceChannel=ECC_GameTraceChannel11
  occlusionVolumeAttenuation=0.5
  reverbWetLevelMin=0.6905
  stereoSpread=100
  attenuationShapeExtents=<650 190 400>
 }
 concurrencyOverrides {maxCount=3, volumeScaleMode=Distance}
}
