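# Looping ambient prop SoundCue for the front of the SilverPopCorn wind machine.
# Two pitch-modulated fan loops, airy industrial louder than air duct, feed two attenuation branches: one with stereo spread, one point-source with occlusion.
# Both branches are mixed and pitch-modulated again, identically for all quality levels.
# The cue also overrides attenuation with an occluded cone and limits concurrency by distance-based volume scaling.
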SoundCue /Hera/Audio/Props/SC_Spot_SilverPopCorn_WindMachine_A_Front
root=QualityLevel_0
nodes [
 {
  id=Attenuation_1
  children [Mixer_2]
  properties {
   bOverrideAttenuation=true
   attenuationOverrides {
    bEnableListenerFocus=true
    falloffDistance=1750
    focusAzimuth=25
    nonFocusAzimuth=55
    nonFocusVolumeAttenuation=0.75
    stereoSpread=50
    attenuationShapeExtents=<250 60 30>
    customAttenuationCurve {editorCurveData {keys [<at 0 1>,<at 0.4985 0.2>,<at 1 0>]}}
    submixSendSettings [{
      bEnableHPFCutoff=false
      bEnableLPFCutoff=false
      disableManualSendClamp=false
      hPFCutoff=20
      lPFCutoff=20000
      maxSendDistance=6000
      maxSendLevel=1
      minSendDistance=400
      minSendLevel=0
      sendLevel=0.2
      sendLevelControlMethod=Linear
      soundSubmix=null
     }]
   }
  }
 }
 {
  id=Attenuation_3
  children [Mixer_2]
  properties {
   bOverrideAttenuation=true
   attenuationOverrides {
    bEnableListenerFocus=true
    bEnableOcclusion=true
    falloffDistance=1750
    focusAzimuth=15
    nonFocusAzimuth=45
    nonFocusVolumeAttenuation=0.75
    occlusionLowPassFilterFrequency=2000
    occlusionTraceChannel=ECC_GameTraceChannel11
    stereoSpread=0
    attenuationShapeExtents=<0 90>
    customAttenuationCurve {editorCurveData {keys [<at 0 1>,<at 0.4985 0.2>,<at 1 0>]}}
    submixSendSettings [{
      bEnableHPFCutoff=false
      bEnableLPFCutoff=false
      disableManualSendClamp=false
      hPFCutoff=20
      lPFCutoff=20000
      maxSendDistance=6000
      maxSendLevel=1
      minSendDistance=400
      minSendLevel=0
      sendLevel=0.2
      sendLevelControlMethod=Linear
      soundSubmix=null
     }]
   }
  }
 }
 {id=Mixer_1, children [Attenuation_1, Attenuation_3], properties {inputVolume [1,1]}}
 {id=Mixer_2, children [Modulator_3,Modulator_2], properties {inputVolume [2,0.3]}}
 {id=Modulator_1, children [Mixer_1], properties {pitchMax=1.1, pitchMin=1}}
 {id=Modulator_2, children [WavePlayer_4], properties {pitchMax=1, pitchMin=0.9, volumeMax=1, volumeMin=1}}
 {id=Modulator_3, children [WavePlayer_3], properties {pitchMax=0.9, pitchMin=0.8, volumeMax=1, volumeMin=1}}
 {id=QualityLevel_0, children [Modulator_1,Modulator_1,Modulator_1,Modulator_1]}
 {
  id=WavePlayer_3
  properties {bLooping=true, soundWaveAssetPtr=/AmbientAudioContent/Waves/SpottedLoops/Spot_Industrial_Processing_Airy_Fan_Loop_02}
 }
 {id=WavePlayer_4, properties {bLooping=true, soundWaveAssetPtr=/AmbientAudioContent/Waves/SpottedLoops/Spot_AMB_Air_Duct_Fan_Loop_01}}
]
properties {
 bEnableSubmixSends=false
 bOverrideAttenuation=true
 bOverrideConcurrency=true
 maxDistance=2000
 soundClassObject=&Fort_Ambient_Props_Athena
 volumeMultiplier=0.5
 attenuationOverrides {
  attenuationShape=Cone
  bEnableListenerFocus=true
  bEnableOcclusion=true
  coneSphereFalloffDistance=350
  coneSphereRadius=250
  falloffDistance=1300
  focusVolumeAttenuation=1.1
  nonFocusVolumeAttenuation=0.6
  occlusionInterpolationTime=1
  occlusionLowPassFilterFrequency=750
  occlusionTraceChannel=ECC_GameTraceChannel11
  occlusionVolumeAttenuation=0.5
  reverbWetLevelMin=0.6905
  stereoSpread=100
  attenuationShapeExtents=<650 90 400>
 }
 concurrencyOverrides {maxCount=3, volumeScaleMode=Distance}
}
