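# Sound cue named as a looping spot ambience for a single giant redwood tree prop, though it holds no sound nodes or root, so it plays nothing as saved.
# It carries only playback setup: redwood-specific attenuation, the outdoor ambient sound class and a send to the exterior ambient source bus.
# An overridden concurrency rule caps simultaneous instances at a few, with a timed release for stolen voices.
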SoundCue /Hera/Audio/Props/SC_Spot_Tree_Giant_Redwood_Loop
root=null
nodes []
properties {
 attenuationSettings=/Hera/Audio/Attenuations/ATT_Hera_Spot_Giant_Redwood_Tree_Single
 bOverrideConcurrency=true
 maxDistance=10700
 soundClassObject=&Fort_Ambient_Outdoors_Athena
 volumeMultiplier=1
 busSends [{audioBus=null, minSendLevel=1, sendLevel=1, soundSourceBus=&AmbientAudio_Exterior_SourceBus}]
 concurrencyOverrides {maxCount=3, voiceStealReleaseTime=2}
}
